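# Physics asset for the face accessory mesh of the Dark Heart female medium player cosmetic, defining bodies for dynamic hair.
# A six-bone hair chain hangs from the head through a master bone, and a three-bone top-hair chain attaches directly to the head.
# Capsules on head and left clavicle plus an angled upper-spine box form the body shapes around the hair.
# Hair segments use limited swing and twist, while master, head and clavicle joints rotate freely.
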PhysicsAsset /BRCosmetics/Characters/Player/Female/Medium/Bodies/F_MED_Dark_Heart/Meshes/Parts/F_MED_Dark_Heart_FaceAcc_Physics
bodies {
 clavicle_l [{shapeType=Capsule, length=7.8314, radius=5.8767, center=<4.1941 -0.0001 0.8856>, rotation=<rot 90 0 0>}]
 dyn_hair_01 [{shapeType=Capsule, length=3.979, radius=2.5327, center=<0.1 0 0>, rotation=<rot 90 0 0>}]
 dyn_hair_02 [{shapeType=Capsule, length=5.6009, radius=2.4181, center=<0.1 0 0>, rotation=<rot 90 0 0>}]
 dyn_hair_03 [{shapeType=Capsule, length=2.7838, radius=2.3887, center=<0.1 0 0>, rotation=<rot 90 0 0>}]
 dyn_hair_04 [{shapeType=Capsule, length=1.8532, radius=2.4206, center=<0.1 0 0>, rotation=<rot 90 0 0>}]
 dyn_hair_05 [{shapeType=Capsule, length=1.1317, radius=2.3983, center=<0.1 0 0>, rotation=<rot 90 0 0>}]
 dyn_hair_06 [{shapeType=Capsule, length=3.291, radius=2.0348, center=<0.1 0 0>, rotation=<rot 90 0 0>}]
 dyn_hair_MSTR [{shapeType=Capsule, length=0.505, radius=0.505}]
 dyn_top_hair_01 [{shapeType=Capsule, length=2.1889, radius=1.3525, center=<1 0 0>, rotation=<rot 90 0 0>}]
 dyn_top_hair_02 [{shapeType=Capsule, length=2.424, radius=1.1174, center=<1 0 0>, rotation=<rot 90 0 0>}]
 dyn_top_hair_03 [{shapeType=Capsule, length=2.416, radius=1.1254, center=<1 0 0>, rotation=<rot 90 0 0>}]
 head [{shapeType=Capsule, length=5.4693, radius=7.7059, center=<3.0863 0.2441 1.7781>, rotation=<rot 90 0 0>}]
 spine_05 [{shapeType=Box, extentX=7.8954, extentY=2.3298, extentZ=20.3892, center=<0.9294 7.8961 -0.0001>, rotation=<rot 0 -35.4578 0>}]
}
constraints [
 {bone1Name=dyn_hair_02, bone2Name=dyn_hair_01, swing1LimitDegrees=45, swing2LimitDegrees=25, twistLimitDegrees=15}
 {bone1Name=dyn_top_hair_01, bone2Name=head, swing1LimitDegrees=30, swing2LimitDegrees=30, twistLimitDegrees=15}
 {bone1Name=dyn_top_hair_02, bone2Name=dyn_top_hair_01, swing1LimitDegrees=45, swing2LimitDegrees=45, twistLimitDegrees=15}
 {bone1Name=dyn_top_hair_03, bone2Name=dyn_top_hair_02, swing1LimitDegrees=45, swing2LimitDegrees=45, twistLimitDegrees=15}
 {bone1Name=dyn_hair_01, bone2Name=dyn_hair_MSTR, swing1LimitDegrees=45, swing2LimitDegrees=45, twistLimitDegrees=15}
 {bone1Name=dyn_hair_03, bone2Name=dyn_hair_02, swing1LimitDegrees=45, swing2LimitDegrees=25, twistLimitDegrees=15}
 {bone1Name=dyn_hair_04, bone2Name=dyn_hair_03, swing1LimitDegrees=45, swing2LimitDegrees=45, twistLimitDegrees=15}
 {bone1Name=dyn_hair_05, bone2Name=dyn_hair_04, swing1LimitDegrees=45, swing2LimitDegrees=45, twistLimitDegrees=15}
 {bone1Name=dyn_hair_06, bone2Name=dyn_hair_05, swing1LimitDegrees=45, swing2LimitDegrees=45, twistLimitDegrees=15}
 {
  bone1Name=dyn_hair_MSTR
  bone2Name=head
  swing1LimitDegrees=45
  swing1Motion=Free
  swing2LimitDegrees=45
  swing2Motion=Free
  twistLimitDegrees=45
  twistMotion=Free
 }
 {
  bone1Name=clavicle_l
  bone2Name=spine_05
  swing1LimitDegrees=45
  swing1Motion=Free
  swing2LimitDegrees=45
  swing2Motion=Free
  twistLimitDegrees=45
  twistMotion=Free
 }
 {
  bone1Name=head
  bone2Name=spine_05
  swing1LimitDegrees=45
  swing1Motion=Free
  swing2LimitDegrees=45
  swing2Motion=Free
  twistLimitDegrees=45
  twistMotion=Free
 }
]
